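# Niagara burst effect for the destruction of the WildEstate BushHerb plant.
# GPU twig meshes and leaf meshes spawn in a sphere, get pushed outward by a point force, then fall and shrink away, with wind and aerodynamic drag on the leaves.
# Red drip sprites vary by spawn group, with the largest burst disabled.
# User.bIsCurie drives leaf material parameters; twig collision and the bIsInstantDeath-driven leaf lift are disabled.
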NiagaraSystem /WildEstateConsumables/Gameplay/BushHerb/Assets/FX/NS_WildEstate_BushHerb_Plant_Destruction
dependencies {
 usedDataInterfaces [
  &NiagaraDataInterfaceMeshRendererInfo
  &NiagaraDataInterfaceCurve
  &NiagaraDataInterfaceArrayFloat3
  /Script/ChaosNiagara.NiagaraDataInterfacePhysicsField
  &NiagaraDataInterfaceArrayInt32
  &NiagaraDataInterfaceArrayFloat
  &NiagaraDataInterfaceSpriteRendererInfo
  &NiagaraDataInterfaceVector2DCurve
 ]
 usedDynamicInputs [
  &MakeVector
  &SelectFloatFromArray
  &self:ScaleFactor
  &ClampFloat
  &Divide_Float
 ]
 usedModules [
  &self:SystemSpawnScript.SystemScriptSource.SystemScriptGraph.NiagaraNodeAssignment_0.SetVariables_#1
  &SystemState
  &self:SystemSpawnScript.SystemScriptSource.SystemScriptGraph.NiagaraNodeAssignment_1.SetVariables_#2
 ]
}
emitters {
 Branches_Twigs {
  simTarget=GPUComputeSim
  emitterUpdateScript [
   &EmitterState()
   &SpawnBurst_Instantaneous(
    Spawn Count=&RandomRangeInt(Minimum=8, Maximum=12)
    Use Spawn Probability=false
    UseLoopCountLimit=false
   )
  ]
  particleSpawnScript [
   &InitializeParticle(
    Color Mode=::NewEnumerator0
    Lifetime Max=2.5
    Lifetime Min=1.5
    Lifetime Mode=::NewEnumerator1
    Mesh Renderer Array Visibility Mode=::NewEnumerator3
    Mesh Renderer Info=DataInterface() {propertyValues {MeshRenderer=&self:Branches_Twigs_2.NiagaraMeshRendererProperties_0}}
    Mesh Scale Mode=::NewEnumerator1
    Mesh Uniform Scale Max=0.6
    Mesh Uniform Scale Min=0.35
    UsePositionOffset=false
    Write Material Random=true
   )
   &ShapeLocation(MaskSpawnGroup=false, Offset=<0 0 100>, Offset Mode=::NewEnumerator0, Override Seed=false, Sphere Radius=50, UseOwnerScale=true) {module=ShapeLocation001}
   &InitialMeshOrientation(Use Rotation Vector=false)
   &PointForce(
    Force Origin=&AddVectorToPosition(Position=/Niagara/DynamicInputs/Helpers/SimulationPosition(), Vector=<0 0 25>)
    Force Strength=&RandomRangeFloat(Minimum=300, Maximum=400)
    Module.Use Falloff=false
    Module.UseLinearForce=false
   )
   &ApplyInitialForces()
  ]
  particleUpdateScript [
   &ParticleState()
   &ScaleMeshSize(Scale Factor=&VectorFromFloat(Value=&FloatFromCurve(FloatCurve=DataInterface() {propertyValues {Curve {keys [<at 0.8 1 RCIM_Cubic>,<at 1 0 RCIM_Cubic>]}}})))
   &GravityForce()
   &MeshRotationForce(Rotation=&RandomRangeVector(Minimum=<1 1 1>, Maximum=<5 5 5>))
   &Collision(
    Age Colliding Particles=false
    Collision Enabled=&SetBoolByFloatComparison(A=$Particles.Age, B=0.1)
    Control Roll On Collision=true
    Friction=0.333
    Friction During a Bounce=0.3
    GPU Collision Type=::NewEnumerator2
    Mesh Dimensions:Vector3f=&SelectVectorFromArray(
     Direct Array Index=$Particles.MeshIndex
     Vector Selection Array=DataInterface() {propertyValues {FloatData [<73 20 18>, <56 6 11>, <47 8 8>]}}
    )
    Module.Control Rest State Via Surface Normal=false
    Module.EnableMaxCollisionCount=false
    Module.Randomize Collision Normal Vector=true
    Radius Calculation Type=::NewEnumerator1
    Restitution=0
    ScaleRateOfRotation=false
   ) {enabled=false}
   &SolveForcesAndVelocity(Clamp Velocity=false, Limit Acceleration=false)
  ]
  rendererClasses [&NiagaraMeshRendererProperties]
  renderers [{rendererClass=&NiagaraMeshRendererProperties, rendererIndex=0}]
 }
 Drips {
  emitterUpdateScript [
   &EmitterState()
   &SpawnBurst_Instantaneous(Spawn Count=50, Use Spawn Probability=false, UseLoopCountLimit=false) {enabled=false}
   &SpawnBurst_Instantaneous(Spawn Count=2, Spawn Group=1, Use Spawn Probability=false, UseLoopCountLimit=false) {module=SpawnBurst_Instantaneous002}
   &SpawnBurst_Instantaneous(Spawn Count=5, Spawn Group=2, Use Spawn Probability=false, UseLoopCountLimit=false) {module=SpawnBurst_Instantaneous001}
  ]
  particleSpawnScript [
   /SetVariables(
    Particles.Variance:NiagaraFloat=&RandomRangeFloat(Minimum=0, Maximum=1)
    Particles.VisibilityTag=&SelectIntFromArray(Direct Array Index=&SpawnGroup(), Int Selection Array=DataInterface() {propertyValues {IntData [0,1,1]}})
   ) {isSetParameters=true}
   &InitializeParticle(
    Color=&MaskLinearColorBySpawnGroup(
     Masked Linear Color=<rgba 1 0 0 1>
     Passthrough Linear Color=&RandomRangeLinearColor(Minimum=<rgba 1 0 0 1>, Maximum=<rgba 1 0 0 1>)
    )
    Lifetime=&RandomRangeFloat(Minimum=0.45, Maximum=0.65)
    Position Offset=<0 0 100>
    Sprite Rotation Mode=::NewEnumerator2
    Sprite Size=&MaskVector2dBySpawnGroup(
     Masked Vector2D=&Lerp_Vector2(A=<1.5 4.5>, B=<3 15>, Alpha=&RandomRangeFloat(Minimum=0, Maximum=1))
     Passthrough Vector2D=&MaskVector2dBySpawnGroup(
      Masked Vector2D=&Vector2DFromFloat(Value=120)
      Passthrough Vector2D=&Vector2DFromFloat(Value=&Lerp_Float(A=60, B=120, Alpha=$Particles.Variance))
      Spawn Group Mask=1
     )
     Spawn Group Mask=0
    )
    Sprite Size Mode=::NewEnumerator4
    UsePositionOffset=true
    Write Material Random=true
   )
   &CameraOffset(Camera Offset Amount=65)
   &AddVelocity(
    Rotation Coordinate Space=::World
    Velocity=&RandomRangeVector(Minimum=<-250 -250 200>, Maximum=<250 250 400>)
   )
   &SubUVAnimation(
    "Sprite Renderer "=DataInterface() {propertyValues {SpriteRenderer=&self:Drips_0.NiagaraSpriteRendererProperties_2}}
    SubUV Animation Mode:ENiagara_SubUVLookupModeV2=::NewEnumerator2
    UseEndFrame=false
    UseStartFrame=false
   )
  ]
  particleUpdateScript [
   &ParticleState()
   &ScaleSpriteSize(
    Scale Factor=&MaskVector2dBySpawnGroup(
     Masked Vector2D=&Vector2DFromCurve(
      Vector2Curve=DataInterface() {propertyValues {
        XCurve {keys [<at 0.5 1 RCIM_Cubic>, {arriveTangent=-0.4016, tangentMode=RCTM_User, time=1, value=0.25}]}
        YCurve {keys [
          {arriveTangent=3.3463, interpMode=RCIM_Cubic, leaveTangent=3.3463, tangentMode=RCTM_User, time=0, value=0}
          <at 0.3 1>
          {arriveTangent=-1.281, interpMode=RCIM_Cubic, leaveTangent=-1.281, tangentMode=RCTM_User, time=0.5, value=1}
          {arriveTangent=-0.1464, tangentMode=RCTM_User, time=1, value=0}
         ]}
       }}
      ScaleCurve=<1 1>
     )
     Passthrough Vector2D=&Vector2DFromFloat(Value=&FloatFromCurve(FloatCurve=DataInterface() {propertyValues {Curve {keys [
           {arriveTangent=4.1785, interpMode=RCIM_Cubic, leaveTangent=4.1785, tangentMode=RCTM_User, time=0, value=0.15}
           {arriveTangent=1.1179, interpMode=RCIM_Cubic, leaveTangent=1.1179, tangentMode=RCTM_User, time=0.2042, value=0.6228}
           {arriveTangent=0.0224, interpMode=RCIM_Cubic, leaveTangent=0.0224, tangentMode=RCTM_User, time=1, value=1}
          ]}}}))
     Spawn Group Mask=0
    )
    Scale Sprite Size Mode=::NewEnumerator2
   )
   &ScaleColor(
    Scale Alpha=&MaskFloatBySpawnGroup(
     Masked Float=&FloatFromCurve(
      FloatCurve=DataInterface() {propertyValues {Curve {keys [
          {arriveTangent=-0.2723, interpMode=RCIM_Cubic, leaveTangent=-0.2723, tangentMode=RCTM_User, time=0, value=1}
          {arriveTangent=-3.7547, interpMode=RCIM_Cubic, leaveTangent=-3.7547, tangentMode=RCTM_User, time=1, value=0}
         ]}}}
      Scale Curve=2
     )
     Passthrough Float=1
     Spawn Group Mask=2
    )
    ScaleA=true
    ScaleRGB=false
   )
   &Drag(
    Drag=&MaskFloatBySpawnGroup(Masked Float=1, Passthrough Float=0)
    Use Linear Drag=true
    Use Rotational Drag=false
   )
   &GravityForce()
   &SolveForcesAndVelocity(Clamp Velocity=false, Limit Acceleration=false)
   &DynamicMaterialParameters(
    Index 0 Param 1=&FloatFromCurve(FloatCurve=DataInterface() {propertyValues {Curve {keys [
         {arriveTangent=-0.0467, interpMode=RCIM_Cubic, leaveTangent=-0.0467, tangentMode=RCTM_User, time=0.1, value=1}
         {arriveTangent=-1.2111, tangentMode=RCTM_User, time=1, value=0}
        ]}}})
    Module.Index0Param0Write=true
    Module.Index0Param1Write=false
    Module.Index0Param2Write=false
    Module.Index0Param3Write=false
   )
  ]
  rendererClasses [&NiagaraSpriteRendererProperties]
  renderers [
   {rendererClass=&NiagaraSpriteRendererProperties}
   {rendererClass=&NiagaraSpriteRendererProperties, rendererIndex=1}
  ]
 }
 Leaves_Mobile {
  emitterUpdateScript [
   &EmitterState()
   &SpawnBurst_Instantaneous(Spawn Count=25, Use Spawn Probability=false, UseLoopCountLimit=false)
  ]
  particleSpawnScript [
   &InitializeParticle(
    Color=<rgba 0.1845 0.2232 0.0369 1>
    Lifetime Max=2.5
    Lifetime Min=1.5
    Lifetime Mode=::NewEnumerator1
    Mesh Renderer Array Visibility Mode=::NewEnumerator3
    Mesh Renderer Info=DataInterface() {propertyValues {MeshRenderer=&self:Leaves_Mobile_2.NiagaraMeshRendererProperties_0}}
    Mesh Scale Mode=::NewEnumerator1
    Mesh Uniform Scale Min=1.5
    UsePositionOffset=false
    Write Material Random=true
   )
   &ShapeLocation(MaskSpawnGroup=false, Offset=<0 0 100>, Offset Mode=::NewEnumerator0, Override Seed=false, Sphere Radius=50, UseOwnerScale=true) {module=ShapeLocation001}
   &InitialMeshOrientation(Use Rotation Vector=false)
   &PointForce(
    Force Origin=&AddVectorToPosition(Position=&SimulationPosition(), Vector=<0 0 25>)
    Force Strength=&FloatFromCurve(
     FloatCurve=DataInterface() {propertyValues {Curve {
        defaultValue=300
        keys [
         {interpMode=RCIM_Cubic, tangentMode=RCTM_User, time=0.25, value=350}
         {arriveTangent=630.6082, interpMode=RCIM_Cubic, leaveTangent=630.6082, tangentMode=RCTM_User, time=1, value=500}
        ]
       }}}
     CurveIndex=&RandomRangeFloat(Minimum=0, Maximum=1)
     Scale Curve=1.5
    )
    Module.Use Falloff=false
    Module.UseLinearForce=false
   )
   &LinearForce(Force=&Multiply_VectorByFloat(
     Vector=<0 0 1>
     Float=&MakeCustomFloatFromBool(
      Bool=$User.bIsInstantDeath
      True Float=&RandomRangeFloat(Minimum=100, Maximum=300)
      False Float=&RandomRangeFloat(Minimum=100, Maximum=200)
     )
    )) {enabled=false, module=LinearForce001}
   &ApplyInitialForces()
  ]
  particleUpdateScript [
   &ParticleState()
   &ScaleMeshSize(Scale Factor=&VectorFromFloat(Value=&FloatFromCurve(FloatCurve=DataInterface() {propertyValues {Curve {keys [<at 0.8 1 RCIM_Cubic>,<at 1 0 RCIM_Cubic>]}}})))
   &GravityForce()
   &WindForce(
    Ground Mask Elevation Mode:ENiagaraWindGroundMaskMode=::NewEnumerator1
    Use Ground Mask=true
    Use Physics Field=true
    Wind Speed=<1000 0 100>
    Wind Speed Scale=0.1
   )
   &AerodynamicDrag(
    Aerodynamic Drag=0.75
    Aerodynamic Rotational Drag=0.05
    Drag Shape Curve=DataInterface() {propertyValues {Curve {keys [
        {interpMode=RCIM_Cubic, tangentMode=RCTM_User, time=0, value=0.1}
        {arriveTangent=2.0799, interpMode=RCIM_Cubic, leaveTangent=2.0799, tangentMode=RCTM_User, time=1, value=1}
       ]}}}
    Lift Contribution=0.333
    Lift Shape Curve=DataInterface() {propertyValues {Curve {keys [
        {interpMode=RCIM_Cubic, tangentMode=RCTM_User, time=0, value=0}
        {interpMode=RCIM_Cubic, tangentMode=RCTM_User, time=0.5, value=1}
        {tangentMode=RCTM_User, time=1, value=0}
       ]}}}
    Rotation Strength:NiagaraFloat=&RandomRangeFloat(Minimum=3, Maximum=8)
    Use Linear Drag=true
    Use Rotational Drag=true
   )
   &SolveForcesAndVelocity(Clamp Velocity=false, Limit Acceleration=false)
   &DynamicMaterialParameters(
    Index 0 Param 1=&MakeCustomFloatFromBool(
     Bool=$User.bIsCurie
     True Float=&FloatFromCurve(FloatCurve=DataInterface() {propertyValues {Curve {keys [
          {arriveTangent=-9.3075, interpMode=RCIM_Cubic, leaveTangent=-9.3075, tangentMode=RCTM_User, time=0, value=2}
          {arriveTangent=-2.0574, interpMode=RCIM_Cubic, leaveTangent=-2.0574, tangentMode=RCTM_User, time=0.2, value=0.625}
          <at 0.75 0 RCIM_Cubic>
         ]}}})
     False Float=0
    )
    Index 0 Param 2=&MakeCustomFloatFromBool(Bool=$User.bIsCurie, True Float=1, False Float=0)
    Module.Param0WriteEnabled=true
    Module.Param1WriteEnabled=true
    Module.Param2WriteEnabled=false
    Module.Param3WriteEnabled=false
   )
  ]
  rendererClasses [&NiagaraMeshRendererProperties]
  renderers [{rendererClass=&NiagaraMeshRendererProperties, rendererIndex=0}]
 }
}
summary {
 systemName=NS_WildEstate_BushHerb_Plant_Destruction
 emitters [
  {emitterName=Branches_Twigs, simTarget=GPUComputeSim, rendererClasses [&NiagaraMeshRendererProperties]}
  {emitterName=Leaves_Mobile, rendererClasses [&NiagaraMeshRendererProperties]}
  {emitterName=Drips, rendererClasses [&NiagaraSpriteRendererProperties]}
 ]
 userVariables [
  User.bIsCurie {type=NiagaraBool}
  User.bIsInstantDeath {type=NiagaraBool}
  User.Bounds {type=Vector3f, description="", defaultValue {struct=Vector3f, value=<3450 3450 3000>}}
  User.DamageDealt {type=NiagaraFloat, defaultValue=1, description=""}
  User.HitLocation {type=Vector3f, description="", defaultValue {struct=Vector3f, value=<0 0 0>}}
  User.HitVector {type=Vector3f, description="", defaultValue {struct=Vector3f, value=<0 0 0>}}
  User.LeafAmount {type=NiagaraFloat, defaultValue=100, description=""}
  User.LeafBaseColor1 {type=LinearColor, description="", defaultValue {struct=LinearColor, value=<rgba 0.1413 0.2052 0.0366 1>}}
  User.LeafHueVariance {type=NiagaraFloat, defaultValue=-0.02, description=""}
  User.LeafValueVariance {type=NiagaraFloat, defaultValue=0.3, description=""}
  User.NoVariance {type=NiagaraFloat, defaultValue=0, description=""}
  User.Radius {type=NiagaraFloat, defaultValue=1085, description=""}
  User.Resource {type=NiagaraInt32}
  User.Scale {type=Vector3f, description="", defaultValue {struct=Vector3f, value=<1 1 1>}}
  User.TreeCylinderHeight {type=NiagaraFloat, defaultValue=350, description=""}
  User.TreeCylinderOffset {type=Vector3f, description="", defaultValue {struct=Vector3f, value=<0 0 180>}}
  User.TreeCylinderRadius {type=NiagaraFloat, defaultValue=0, description=""}
  User.TreeSpheriodRadius {type=NiagaraFloat, defaultValue=200, description=""}
  User.TrunkCylinderHeight {type=NiagaraFloat, defaultValue=0, description=""}
  User.TrunkCylinderRadius {type=NiagaraFloat, defaultValue=0, description=""}
 ]
}
